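; Door Blueprint variant for the Storage Wars feature that adds a blocking collision mesh, removed when the door activates or opens.
; Active and Open overrides disable that collision, and activation also sets IsDoorOpened on the authority.
; OnRep_IsDoorOpened toggles the blocker to match the replicated state.
; If the StorageWarsEnabledHF scalable float is false, BeginPlay disables the blocker and makes the actor damageable.
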(event EventBeginPlay
  (|CallParentFunction)
  (if (not (ScalableFloat|GetValueasBool (Variables|Default|GetStorageWarsEnabledHF)))
    (CallFunction|SetBlockAllMeshEnabled false)
    (Variables|Actor|SetbCanBeDamaged true)))

(fn BP_OnActive
  (params (EnableAlternateActive :type Boolean) (EnableOpenCues :type Boolean))
  (|CallParentFunction EnableAlternateActive EnableOpenCues)
  (CallFunction|SetBlockAllMeshEnabled false)
  (Networking|SwitchHasAuthority
    (:Authority
      (Variables|Default|SetIsDoorOpened true))))

(fn BP_OnOpen
  (params (EnableAlternateOpen :type Boolean) (EnableOpenCues :type Boolean))
  (|CallParentFunction EnableAlternateOpen EnableOpenCues)
  (CallFunction|SetBlockAllMeshEnabled false))

(fn OnRep_IsDoorOpened
  (if (Variables|Default|GetIsDoorOpened)
    (CallFunction|SetBlockAllMeshEnabled false)
    (else
      (CallFunction|SetBlockAllMeshEnabled true))))

(fn SetBlockAllMeshEnabled
  (params (Enabled :type Boolean))
  (bind _blockallmesh (Variables|Default|GetBlockAllMesh))
  (if Enabled
    (Collision|SetCollisionEnabled _blockallmesh "QueryAndPhysics")
    (else
      (Collision|SetCollisionEnabled _blockallmesh "NoCollision"))))

(fn ConstructionScript
  (|CallParentFunction))
