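; Storage Wars large door trigger that opens a storage room for a gold cost, or acts as a resettable door trigger when hotfix-disabled.
; On the server it activates a random enabled data layer setup, then polls a room overlap box on a timer to find late-loading chests, safes and audio players.
; Clients silence those replicated actors until the door opens.
; Payment uses either standard interaction or a spawned conversation prop.
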(event Custom|DebugStorageWarsSetLayerIndex
  (params (Index :type Integer))
  ; Debug Commands
  (CallFunction|ForceSetDataLayer Index))

(event EventBeginPlay
  (|CallParentFunction)
  (if (ScalableFloat|GetValueasBool (Variables|HotFix|GetStorageWarsEnabledHF))
    (CallFunction|CacheHFVariables)
    (CallFunction|SpawnDataLayer)
    ;[1 0.1317 0 0.5] This actor can be configured to use the Conversation UI, or the stand Interaction UI
    (CallFunction|SpawnConversationPropActor)
    (Networking|SwitchHasAuthority
      (:Authority
        (if (not (Variables|Default|GetTriggerActivated))
          (bind _returnvalue (Utilities|Time|SetTimerbyEvent (EventDispatchers|CreateEvent :!event TryFindOverlappingActors) (Math|Float|Clamp(Float) (ScalableFloat|GetValueatLevel (Variables|HotFix|GetFindOverlappingActorsTickRateHF)) 0.033 (ScalableFloat|GetValueatLevel (Variables|HotFix|GetFindOverlappingActorsTickRateHF))) true))
          (Variables|Default|SetSilenceChestsTimer _returnvalue))))
    (else
      (CallFunction|SetupNoStorageWars))))

(event Custom|SpawnDataLayer
  ; Load random data layer
  (Networking|SwitchHasAuthority
    (:Authority
      (if (> (Utilities|Array|Length (Variables|Default|GetDataLayersList)) 0)
        (if (Variables|Default|GetDebugEnableLayerSpawn)
          (if (Editor|InEditororPIE)
            (Default|ApplyDataLayerSetup (Utilities|Array|Get(acopy) (Variables|Default|GetDataLayersList) (Variables|Default|GetDebugLayerSpawnIndex)))
            (Variables|Default|SetCurrentDataLayerIndex (Variables|Default|GetDebugLayerSpawnIndex))
            (Variables|Default|SetCurrentDataLayerIndexSet true)
            (Default|CallOnCurrentDataLayerIndexSet (Variables|Default|GetCurrentDataLayerIndex))
            (else
              _))
          (else
            _))
        (bind _index (CallFunction|GetDataLayerIndextoUse))
        (if (>= _index 0)
          (Default|ApplyDataLayerSetup (Utilities|Array|Get(acopy) (Variables|Default|GetDataLayersList) _index))
          (Variables|Default|SetCurrentDataLayerIndex _index)
          (Variables|Default|SetCurrentDataLayerIndexSet true)
          (Default|CallOnCurrentDataLayerIndexSet (Variables|Default|GetCurrentDataLayerIndex))
          (else
            (Development|PrintString "blah STORAGE WARS - NO VALID DATA LAYER INDEX, WILL NOT WORK" true true "(R=1.000000,G=0.065051,B=0.000000,A=1.000000)")))))))

(fn ActivateTargetedDoor
  (params (InInstigator :type Object%20Reference) (EnableAlternateOpen :type Boolean))
  (|CallParentFunction InInstigator EnableAlternateOpen)
  (Networking|SwitchHasAuthority
    (:Authority
      (Variables|Default|SetDoorOpened true)
      (Utilities|Time|ClearandInvalidateTimerbyHandle (Variables|Default|GetSilenceChestsTimer))
      (CallFunction|UnsilenceActors)
      (if (not (Variables|Interaction|GetUseConversationUI))
        (bind _fortpawn (Class|FortQuestBlueprintLibrary|GetFortPawnFromObject (Variables|GetInInstigator))
          (:Valid
            (Default|RemoveGold _fortpawn)))))))

(fn ApplyDataLayerSetup
  (params (RandomDataLayerData :type Random%20Data%20Layer%20Data%20Structure%20%28by%20ref%29))
  (bind (_enableddatalayers_9_41bb9a3e4648805b5af70aa728288c92 _disableddatalayers_10_8de59e3b474c88ed2e7988b185572444) (Utilities|Struct|BreakRandomDataLayerData RandomDataLayerData))
  (bind _returnvalue (WorldPartition|GetDataLayerManager))
  (Utilities|FlowControl|Sequence
    (:then_0
      (for _array_element _enableddatalayers_9_41bb9a3e4648805b5af70aa728288c92
        (DataLayers|SetDataLayerInstanceRuntimeState _returnvalue (DataLayers|GetDataLayerInstancefromAsset _returnvalue _array_element) "Activated" false)))
    (:then_1
      (for _array_element_2 _disableddatalayers_10_8de59e3b474c88ed2e7988b185572444
        (DataLayers|SetDataLayerInstanceRuntimeState _returnvalue (DataLayers|GetDataLayerInstancefromAsset _returnvalue _array_element_2) "Unloaded" false)))))

(fn BlueprintCanInteract
  (params (InteractingPawn :type Fort%20Pawn%20Object%20Reference) (InteractionBeingAttempted :type EInteractionBeingAttempted%20Enum :default "FirstInteraction") (InteractionType :type TInteractionType%20Enum :default "IT_NoInteraction"))
  (return (and (Default|HasEnoughGold InteractingPawn) (not (Variables|Default|GetDoorOpened)))))

(fn BlueprintGetFailedInteractionString
  (params (InteractingPawn :type Fort%20Pawn%20Object%20Reference) (InteractionBeingAttempted :type EInteractionBeingAttempted%20Enum :default "FirstInteraction"))
  (if (not (Variables|Default|GetDoorOpened))
    (return (Variables|Interaction|GetFormattedInteractionText))
    (else
      (return ""))))

(fn BlueprintGetInteractionString
  (params (InteractingPawn :type Fort%20Pawn%20Object%20Reference) (InteractionBeingAttempted :type EInteractionBeingAttempted%20Enum :default "FirstInteraction"))
  (return (Variables|Interaction|GetFormattedInteractionText)))

(fn CacheHFVariables
  ; If we don't use the Conversation UI, we will default to normal Interaction UI
  (Variables|Interaction|SetUseConversationUI (ScalableFloat|GetValueasBool (Variables|HotFix|GetUseConversationUIHF)))
  (Variables|Interaction|SetGoldCost (Math|Float|Truncate (ScalableFloat|GetValueatLevel (Variables|HotFix|GetGoldCostHF))))
  (Variables|Interaction|SetFormattedInteractionText (CallFunction|GetInteractionTextwithCost)))

(fn CanActorBeSilenced
  (params (Actor :type Actor%20Object%20Reference))
  ; Gross, but we have multiple BP parents that do the same thing.
  (Utilities|Casting|CastToTiered_Chest_Athena_Parent Actor
    (:then
      _)
    (:CastFailed
      (Utilities|Casting|CastToTiered_Chest_Athena Actor
        (:then
          _)
        (:CastFailed
          (Utilities|Casting|CastToTiered_Safe_Athena_Parent Actor
            (:then
              _)
            (:CastFailed
              (Utilities|Casting|CastToBP_DragonCart_StorageWars_AudioPlayer Actor
                (:then
                  _)
                (:CastFailed
                  (return false)))))))))
  (return true))

(fn ForceSetDataLayer
  (params (Index :type Integer))
  (if (Utilities|Array|IsValidIndex (Variables|Default|GetDataLayersList) (Variables|GetIndex))
    (Development|PrintString (Utilities|String|Append "STORAGE WARS LOADING LAYER INDEX " (Utilities|String|ToString(Integer) (Variables|GetIndex))) true true "(R=1.000000,G=0.174434,B=0.000000,A=1.000000)" 5.0)
    (Utilities|FlowControl|Sequence
      (:then_0
        (bind _returnvalue (WorldPartition|GetDataLayerManager))
        (for _array_element (Variables|Default|GetDataLayersList)
          (for _array_element_2 (Utilities|Struct|BreakRandomDataLayerData _array_element)
            (DataLayers|SetDataLayerInstanceRuntimeState _returnvalue (DataLayers|GetDataLayerInstancefromAsset _returnvalue _array_element_2) "Unloaded" false))))
      (:then_1
        (Default|ApplyDataLayerSetup (Utilities|Array|Get(acopy) (Variables|Default|GetDataLayersList) (Variables|GetIndex)))))
    (else
      (Development|PrintString "STORAGE WARS CANNOT SET LAYER, INDEX OUT OF BOUNDS" true true "(R=1.000000,G=0.174434,B=0.000000,A=1.000000)" 5.0))))

(fn GetBoundsComponents
  (params (SelfActor :type Actor%20Object%20Reference))
  (return (Utilities|Array|MakeArray (Variables|Default|GetBox))))

(fn GetDataLayerIndexToUse
  (for (_array_element _array_index) (Variables|Default|GetDataLayersList)
    (LocalVariables|SetTempIndex _array_index)
    (if (and (Utilities|Array|IsNotEmpty (Utilities|Struct|BreakRandomDataLayerData _array_element)) (ScalableFloat|GetValueasBool (.bSetupEnabled_22_436E2FE243B567AE5E7CA9AAC2346CA0 (Utilities|Struct|BreakRandomDataLayerData _array_element))))
      (Utilities|Array|Add (LocalVariables|GetValidIndexes) (LocalVariables|GetTempIndex))))
  (if (Utilities|Array|IsNotEmpty (LocalVariables|GetValidIndexes))
    (return (Utilities|Array|RandomArrayItem (LocalVariables|GetValidIndexes)))
    (else
      (return -1))))

(fn GetInteractionTextWithCost
  (return (Utilities|Text|FormatText "NSLOCTEXT(\"[C8E87623B90DA05153998483C99D6925]\", \"70A8317E463A6DE84EBCCDB4F80F245B\", \"{activate} ({goldcost} {gold})\")" (Variables|Interaction|GetActivateDoorText) (Variables|Interaction|GetGoldCost) (Variables|Interaction|GetGoldText))))

(fn HasActorBeenSearched
  (params (Actor :type Actor%20Object%20Reference))
  (bind _astiered_chest_athena_parent (Utilities|Casting|CastToTiered_Chest_Athena_Parent Actor)
    (:then
      (return (Variables|/Script/FortniteGame.BuildingContainer|GetbAlreadySearched _astiered_chest_athena_parent)))
    (:CastFailed
      (bind _astiered_chest_athena (Utilities|Casting|CastToTiered_Chest_Athena Actor)
        (:then
          (return (Variables|/Script/FortniteGame.BuildingContainer|GetbAlreadySearched _astiered_chest_athena)))
        (:CastFailed
          (bind _astiered_safe_athena_parent (Utilities|Casting|CastToTiered_Safe_Athena_Parent Actor)
            (:then
              (return (Variables|/Script/FortniteGame.BuildingContainer|GetbAlreadySearched _astiered_safe_athena_parent)))
            (:CastFailed
              (return false))))))))

(fn HasEnoughGold
  (params (PlayerPawn :type Fort%20Pawn%20Object%20Reference))
  (bind _returnvalue (Pawn|GetController PlayerPawn))
  (Utilities|IsValid PlayerPawn
    (:Is%20Valid
      (Utilities|IsValid _returnvalue
        (:Is%20Valid
          (bind _asfort_player_controller_athena (Utilities|Casting|CastToFortPlayerControllerAthena _returnvalue)
            (:then
              (bind _returnvalue_2 (InventoryManagement|GetItemCountforItemDefinition _asfort_player_controller_athena (Variables|Interaction|GetGoldItemDefinition)))
              (return (>= _returnvalue_2 (Variables|Interaction|GetGoldCost))))
            (:CastFailed
              _)))
        (:Is%20Not%20Valid
          _))
      (return false))
    (:Is%20Not%20Valid
      _)))

(fn OnRep_DoorOpened
  (if (Variables|Default|GetDoorOpened)
    (Components|DestroyComponent (Variables|Default|GetRoomOverlapBox))
    (Networking|SwitchHasAuthority
      (:Remote
        (CallFunction|UnsilenceActors)
        (Utilities|Time|ClearandInvalidateTimerbyHandle (Variables|Default|GetSilenceChestsTimer))))))

(fn OnRep_OverlappingActorsToSilence
  (Networking|SwitchHasAuthority
    (:Remote
      (if (not (Variables|Default|GetDoorOpened))
        (for _array_element (Variables|Default|GetOverlappingActorsToSilence)
          (bind _output_get (LocalVariables|SetTempActor _array_element))
          (Utilities|IsValid _output_get
            (:Is%20Valid
              (if (not (CallFunction|HasActorBeenSearched _output_get))
                (CallFunction|SetActorSilenced (LocalVariables|GetTempActor) true)))))))))

(fn RemoveGold
  (params (InteractingPlayer :type Fort%20Pawn%20Object%20Reference))
  (bind _returnvalue (Pawn|GetController InteractingPlayer))
  (Networking|SwitchHasAuthority
    (:Authority
      (Utilities|IsValid InteractingPlayer
        (:Is%20Valid
          (Utilities|IsValid _returnvalue
            (:Is%20Valid
              (bind _asfort_player_controller_athena (Utilities|Casting|CastToFortPlayerControllerAthena _returnvalue))
              (bind _returnvalue_2 (InventoryManagement|FindExistingItemforDefinition _asfort_player_controller_athena (Variables|Interaction|GetGoldItemDefinition)))
              (Utilities|IsValid _returnvalue_2
                (:Is%20Valid
                  (Inventory|RemoveFortItemFromPlayer _asfort_player_controller_athena _returnvalue_2 (Variables|Interaction|GetGoldCost)))))))))))

(fn SetActorSilenced
  (params (Actor :type Actor%20Object%20Reference) (Silenced :type Boolean))
  ; Enable / Disable FX from chests
  (if (Variables|GetSilenced)
    (bind _astiered_chest_athena_parent (Utilities|Casting|CastToTiered_Chest_Athena_Parent (Variables|GetActor))
      (:then
        (Networking|SwitchHasAuthority
          (:Remote
            (Audio|Components|Audio|SetPaused (Variables|/Game/Building/ActorBlueprints/Containers/Tiered_Chest_Athena_Parent.Tiered_Chest_Athena_Parent_C|GetChest_Ambient_Sound _astiered_chest_athena_parent) true)
            (Class|TieredChestAthenaParent|SetLightVisibility _astiered_chest_athena_parent false)
            (if (Variables|/Game/Building/ActorBlueprints/Containers/Tiered_Chest_Athena_Parent.Tiered_Chest_Athena_Parent_C|GetEnableAudioIndicator _astiered_chest_athena_parent)
              (Class|BuildingContainer|StopAudiovisualIndicator _astiered_chest_athena_parent)))))
      (:CastFailed
        (bind _astiered_chest_athena (Utilities|Casting|CastToTiered_Chest_Athena (Variables|GetActor))
          (:then
            (Networking|SwitchHasAuthority
              (:Remote
                (Audio|Components|Audio|SetPaused (Variables|/Game/Building/ActorBlueprints/Containers/Tiered_Chest_Athena.Tiered_Chest_Athena_C|GetChest_Ambient_Sound _astiered_chest_athena) true)
                (Class|TieredChestAthena|SetLightVisibility _astiered_chest_athena false)
                (if (Variables|/Game/Building/ActorBlueprints/Containers/Tiered_Chest_Athena.Tiered_Chest_Athena_C|GetEnableAudioIndicator _astiered_chest_athena)
                  (Class|BuildingContainer|StopAudiovisualIndicator _astiered_chest_athena)))))
          (:CastFailed
            (bind _astiered_safe_athena_parent (Utilities|Casting|CastToTiered_Safe_Athena_Parent (Variables|GetActor))
              (:then
                (Networking|SwitchHasAuthority
                  (:Remote
                    (Audio|Components|Audio|SetPaused (Variables|/Game/Building/ActorBlueprints/Containers/Tiered_Safe_Athena_Parent.Tiered_Safe_Athena_Parent_C|GetChest_Ambient_Sound _astiered_safe_athena_parent) true)
                    (Class|TieredSafeAthenaParent|SetLightVisibility _astiered_safe_athena_parent false)
                    (if (Variables|/Game/Building/ActorBlueprints/Containers/Tiered_Safe_Athena_Parent.Tiered_Safe_Athena_Parent_C|GetEnableAudioIndicator _astiered_safe_athena_parent)
                      (Class|BuildingContainer|StopAudiovisualIndicator _astiered_safe_athena_parent)))))
              (:CastFailed
                (bind _asbp_dragon_cart_storage_wars_audio_player (Utilities|Casting|CastToBP_DragonCart_StorageWars_AudioPlayer (Variables|GetActor)))
                (Networking|SwitchHasAuthority
                  (:Remote
                    (Class|BPDragonCartStorageWarsAudioPlayer|SetAudioEnabled _asbp_dragon_cart_storage_wars_audio_player false 0.0)))))))))
    (else
      (bind _astiered_chest_athena_parent_2 (Utilities|Casting|CastToTiered_Chest_Athena_Parent (Variables|GetActor))
        (:then
          (Networking|SwitchHasAuthority
            (:Remote
              (Audio|Components|Audio|SetPaused (Variables|/Game/Building/ActorBlueprints/Containers/Tiered_Chest_Athena_Parent.Tiered_Chest_Athena_Parent_C|GetChest_Ambient_Sound _astiered_chest_athena_parent_2) false)
              (Class|TieredChestAthenaParent|SetLightVisibility _astiered_chest_athena_parent_2 true)
              (if (Variables|/Game/Building/ActorBlueprints/Containers/Tiered_Chest_Athena_Parent.Tiered_Chest_Athena_Parent_C|GetEnableAudioIndicator _astiered_chest_athena_parent_2)
                (Class|BuildingContainer|StartAudiovisualIndicator _astiered_chest_athena_parent_2 (Variables|/Game/Building/ActorBlueprints/Containers/Tiered_Chest_Athena_Parent.Tiered_Chest_Athena_Parent_C|GetChest_Ambient_Sound _astiered_chest_athena_parent_2) "Chest" (Variables|/Game/Building/ActorBlueprints/Containers/Tiered_Chest_Athena_Parent.Tiered_Chest_Athena_Parent_C|GetChestChimeVisualUpdate _astiered_chest_athena_parent_2))))))
        (:CastFailed
          (bind _astiered_chest_athena_2 (Utilities|Casting|CastToTiered_Chest_Athena (Variables|GetActor))
            (:then
              (Networking|SwitchHasAuthority
                (:Remote
                  (Audio|Components|Audio|SetPaused (Variables|/Game/Building/ActorBlueprints/Containers/Tiered_Chest_Athena.Tiered_Chest_Athena_C|GetChest_Ambient_Sound _astiered_chest_athena_2) false)
                  (Class|TieredChestAthena|SetLightVisibility _astiered_chest_athena_2 true)
                  (if (Variables|/Game/Building/ActorBlueprints/Containers/Tiered_Chest_Athena.Tiered_Chest_Athena_C|GetEnableAudioIndicator _astiered_chest_athena_2)
                    (Class|BuildingContainer|StartAudiovisualIndicator _astiered_chest_athena_2 (Variables|/Game/Building/ActorBlueprints/Containers/Tiered_Chest_Athena.Tiered_Chest_Athena_C|GetChest_Ambient_Sound _astiered_chest_athena_2) "Chest" (Variables|/Game/Building/ActorBlueprints/Containers/Tiered_Chest_Athena.Tiered_Chest_Athena_C|GetChestChimeVisualUpdate _astiered_chest_athena_2))))))
            (:CastFailed
              (bind _astiered_safe_athena_parent_2 (Utilities|Casting|CastToTiered_Safe_Athena_Parent (Variables|GetActor))
                (:then
                  (Networking|SwitchHasAuthority
                    (:Remote
                      (Audio|Components|Audio|SetPaused (Variables|/Game/Building/ActorBlueprints/Containers/Tiered_Safe_Athena_Parent.Tiered_Safe_Athena_Parent_C|GetChest_Ambient_Sound _astiered_safe_athena_parent_2) false)
                      (Class|TieredSafeAthenaParent|SetLightVisibility _astiered_safe_athena_parent_2 true)
                      (if (Variables|/Game/Building/ActorBlueprints/Containers/Tiered_Safe_Athena_Parent.Tiered_Safe_Athena_Parent_C|GetEnableAudioIndicator _astiered_safe_athena_parent_2)
                        (Class|BuildingContainer|StartAudiovisualIndicator _astiered_safe_athena_parent_2 (Variables|/Game/Building/ActorBlueprints/Containers/Tiered_Safe_Athena_Parent.Tiered_Safe_Athena_Parent_C|GetChest_Ambient_Sound _astiered_safe_athena_parent_2) "Chest" (Variables|/Game/Building/ActorBlueprints/Containers/Tiered_Safe_Athena_Parent.Tiered_Safe_Athena_Parent_C|GetChestChimeVisualUpdate _astiered_safe_athena_parent_2))))))
                (:CastFailed
                  (bind _asbp_dragon_cart_storage_wars_audio_player_2 (Utilities|Casting|CastToBP_DragonCart_StorageWars_AudioPlayer (Variables|GetActor)))
                  (Networking|SwitchHasAuthority
                    (:Remote
                      (Class|BPDragonCartStorageWarsAudioPlayer|SetAudioEnabled _asbp_dragon_cart_storage_wars_audio_player_2 true 0.0))))))))))))

(fn SetupNoStorageWars
  (Variables|Interaction|SetbAllowInteract true)
  (Variables|Default|SetActivateOnDamaged true)
  (Variables|DragonCartDoor|SetCanResetAfterActivation true)
  (Components|DestroyComponent (Variables|Default|GetRoomOverlapBox)))

(fn SpawnConversationPropActor
  (if (Variables|Interaction|GetUseConversationUI)
    (Collision|SetCollisionEnabled (Variables|Default|GetActionButton) "NoCollision")
    (Rendering|SetVisibility (Variables|Default|GetActionButton) false false)
    (Networking|SwitchHasAuthority
      (:Authority
        (Game|SpawnActorfromClass "/DragonCartDoors/Gameplay/StorageWars/BP_DragonCart_LargeDoorTrigger_StorageWars_ConversationProp.BP_DragonCart_LargeDoorTrigger_StorageWars_ConversationProp_C" (Math|Transform|MakeTransform (Transformation|GetWorldLocation (Variables|Default|GetActionButton)) (Transformation|GetWorldRotation (Variables|Default|GetActionButton))) "Undefined" "MultiplyWithRoot" self 0 0 0)))
    (else
      (Variables|Interaction|SetbAllowInteract true)
      (Variables|Interaction|SetbAllowHighlight true))))

(fn TryFindOverlappingActors
  (Networking|SwitchHasAuthority
    (:Authority
      (if (not (Variables|Default|GetTriggerActivated))
        (bind _outactors (Collision|BoxOverlapActorsWithOrientation (Transformation|GetWorldLocation (Variables|Default|GetRoomOverlapBox)) (Variables|/Script/Engine.BoxComponent|GetBoxExtent (Variables|Default|GetRoomOverlapBox)) (Transformation|GetActorRotation) (Utilities|Array|MakeArray "ObjectTypeQuery1" "ObjectTypeQuery2") 0 (Utilities|Array|MakeArray self)))
        (for _array_element _outactors
          (bind _output_get (LocalVariables|SetTempActor _array_element))
          (Utilities|IsValid _output_get
            (:Is%20Valid
              (if (Utilities|Array|ContainsItem (Variables|Default|GetOverlappingActorsToSilence) (LocalVariables|GetTempActor))
                _
                (elif (CallFunction|CanActorbeSilenced (LocalVariables|GetTempActor))
                  (Utilities|Array|Add (Variables|Default|GetOverlappingActorsToSilence) (LocalVariables|GetTempActor))
                  (Utilities|FlowControl|Sequence
                    (:then_0
                      (if (not (CallFunction|HasActorBeenSearched (LocalVariables|GetTempActor)))
                        (CallFunction|SetActorSilenced (LocalVariables|GetTempActor) true)))
                    (:then_1
                      (bind _asbp_dragon_cart_storage_wars_audio_player (Utilities|Casting|CastToBP_DragonCart_StorageWars_AudioPlayer (LocalVariables|GetTempActor)))
                      (CallFunction|SetActorSilenced _asbp_dragon_cart_storage_wars_audio_player true))))))))
        ; This isn't ideal, but we don't know when the chests will be fully loaded after randomly selecting a data layer, so loop to find them
        (Math|Integer|IncrementInt (Variables|Default|GetTimesLoopedFindingChests)
          (:%20%20
            (if (>= (Variables|Default|GetTimesLoopedFindingChests) (ScalableFloat|GetValueatLevel (Variables|HotFix|GetFindOverlappingActorsTimesToLoopHF)))
              (Utilities|Time|ClearandInvalidateTimerbyHandle (Variables|Default|GetSilenceChestsTimer)))))
        (else
          (Utilities|Time|ClearandInvalidateTimerbyHandle (Variables|Default|GetSilenceChestsTimer)))))))

(fn UnsilenceActors
  (for _array_element (Variables|Default|GetOverlappingActorsToSilence)
    (bind _output_get (LocalVariables|SetTempActor _array_element))
    (Utilities|IsValid _output_get
      (:Is%20Valid
        (if (not (CallFunction|HasActorBeenSearched _output_get))
          (CallFunction|SetActorSilenced (LocalVariables|GetTempActor)))))))

(fn ConstructionScript
  (|CallParentFunction))
