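; Conversation prop for the Storage Wars large dragon cart door, letting a player open the door it is owned by.
; ActivateDoor, on authority, sets DoorOpened, casts the owner to the Storage Wars large door trigger and activates its targeted door with the opening player.
; Interaction is allowed only while DoorOpened is false.
; HideUI, also run from the DoorOpened RepNotify, disables interaction, hides the indicator and deactivates the NonPlayerConversation component.
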(fn ActivateDoor
  (params (OpeningPlayer :type Actor%20Object%20Reference))
  (bind _returnvalue (Actor|GetOwner))
  (Networking|SwitchHasAuthority
    (:Authority
      (Variables|Default|SetDoorOpened true)
      (Utilities|IsValid _returnvalue
        (:Is%20Valid
          (bind _asbp_dragon_cart_large_door_trigger_storage_wars (Utilities|Casting|CastToBP_DragonCart_LargeDoorTrigger_StorageWars _returnvalue))
          (Utilities|IsValid OpeningPlayer
            (:Is%20Valid
              (Class|BPDragonCartLargeDoorTrigger|ActivateTargetedDoor _asbp_dragon_cart_large_door_trigger_storage_wars OpeningPlayer false)
              (CallFunction|HideUI))))))))

(fn BlueprintCanInteract
  (params (InteractingPawn :type Fort%20Pawn%20Object%20Reference) (InteractionBeingAttempted :type EInteractionBeingAttempted%20Enum :default "FirstInteraction") (InteractionType :type TInteractionType%20Enum :default "IT_NoInteraction"))
  (return (not (Variables|Default|GetDoorOpened))))

(fn HideUI
  (bind _nonplayerconversationcomponent (Variables|Components|GetNonPlayerConversationComponent))
  (Variables|Interaction|SetbAllowInteract false)
  (InteractionUI|AddHideUIIndicatorReason _nonplayerconversationcomponent (Utilities|Name|MakeLiteralName "Storage War Door Opened"))
  (Components|Activation|Deactivate _nonplayerconversationcomponent))

(fn OnRep_DoorOpened
  (CallFunction|HideUI))

(fn ConstructionScript
  (|CallParentFunction))
