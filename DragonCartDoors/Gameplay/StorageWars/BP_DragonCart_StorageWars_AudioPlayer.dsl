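; Audio player actor for the DragonCart Storage Wars gameplay, wrapping an audio component that can be faded on and off.
; On BeginPlay it registers that component with the MusicEventSubsystem as a special POI music event, then enables playback if StartAudioEnabled is set.
; SetAudioEnabled fades the component linearly in to full volume or out to silence over FadeDuration.
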(event EventBeginPlay
  (Audio|AddEventbyComponent (GameInstanceSubsystems|GetMusicEventSubsystem) (Variables|Default|GetAudio) "(TagName=\"MusicEventSubsystem.Event.POI.Special\")" "(GameplayTags=)" 0 "")
  (if (Variables|Default|GetStartAudioEnabled)
    (CallFunction|SetAudioEnabled true)))

(fn SetAudioEnabled
  (params (Enabled :type Boolean) (FadeDuration :type Float%20%28double-precision%29))
  (if Enabled
    (Audio|Components|Audio|FadeIn (Variables|Default|GetAudio) (Variables|GetFadeDuration) 1.0 0.0 "Linear" :!class /Script/Engine.AudioComponent)
    (else
      (Audio|Components|Audio|FadeOut (Variables|Default|GetAudio) (Variables|GetFadeDuration) 0.0 "Linear" :!class /Script/Engine.AudioComponent))))

(fn ConstructionScript)
