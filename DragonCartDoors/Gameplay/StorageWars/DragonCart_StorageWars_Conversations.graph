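# Conversation database for the DragonCart Storage Wars mystery door, entered through the DragonCartDoor entry tag.
# A speech node asks whether to open the door and branches to an Activate Door option or a Back option.
# Activation is a data-driven service gated by a HasBars requirement on the player, which stays visible when failed.
# Choosing it runs a side effect that activates the door.
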ConversationDatabase /DragonCartDoors/Gameplay/StorageWars/DragonCart_StorageWars_Conversations
entry_points [{destination_list="[{}]", entry_identifier="", entry_tag {return_value=Conversation.Entry.DragonCartDoor}}]
nodes [
 ConversationEntryPointNode_3 {
  id=#1
  class=ConversationEntryPointNode
  outputs [#2]
  properties {compiled_NodeGUId=#3, entryTag {tagName=Conversation.Entry.DragonCartDoor}}
 }
 Speech_C_3 {
  id=#2
  class=Speech_C
  outputs [#4, #5]
  properties {compiled_NodeGUId=#6, generalConfig {defaultMessage=OPEN THE MYSTERY DOOR?}}
  sub_nodes [ConversationChoiceNode_11 {class=ConversationChoiceNode, properties {defaultChoiceDisplayText=TEST CHOICE}}]
 }
 Continue_C_0 {
  id=#5
  class=Continue_C
  properties {compiled_NodeGUId=#7, softTaskIcon=&Cancel_Normal}
  sub_nodes [ConversationChoiceNode_3 {class=ConversationChoiceNode, properties {defaultChoiceDisplayText=BACK}}]
 }
 DataDrivenService_DragonCart_StorageWars_ActivateDoor_C_2 {
  id=#4
  class=DataDrivenService_DragonCart_StorageWars_ActivateDoor_C
  outputs []
  properties {
   compiled_NodeGUId=#8
   requirements [{
     cannotUseReasonParameter=""
     failureNodeBehaviour=FailedButVisible
     requirement=&self:DataDrivenService_DragonCart_StorageWars_ActivateDoor_C_2.FortControllerRequirement_HasBars_0
     participantId {tagName=Conversation.Participant.Player}
    }]
  }
  sub_nodes [
   ConversationChoiceNode_2 {class=ConversationChoiceNode, properties {defaultChoiceDisplayText=Activate Door}}
   SideEffect_DragonCart_StorageWars_ActivateDoor_C_0 {class=SideEffect_DragonCart_StorageWars_ActivateDoor_C}
  ]
 }
]
speakers [{fallback_name=""}, {fallback_name=""}]
