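; Server-side conversation side effect that activates a Storage Wars large door in DragonCart.
; It takes the NPC and player participant actors from the conversation context and continues only if both are valid.
; It casts the NPC to the large door trigger conversation prop and calls ActivateDoor on it, passing the player.
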(fn ServerCauseSideEffect
  (params (Context :type Conversation%20Context%20Structure%20%28by%20ref%29))
  (bind _returnvalue (Class|FortConversationContextLibrary|GetParticipantActorNPC Context))
  (Utilities|IsValid _returnvalue
    (:Is%20Valid
      (bind _returnvalue_2 (Class|FortConversationContextLibrary|GetParticipantActorPlayer Context))
      (Utilities|IsValid _returnvalue_2
        (:Is%20Valid
          (bind _asbp_dragon_cart_large_door_trigger_storage_wars_conversation_prop (Utilities|Casting|CastToBP_DragonCart_LargeDoorTrigger_StorageWars_ConversationProp _returnvalue))
          (Class|BPDragonCartLargeDoorTriggerStorageWarsConversationProp|ActivateDoor _asbp_dragon_cart_large_door_trigger_storage_wars_conversation_prop _returnvalue_2)
          (return))))))
